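# Physics asset for the BarbequeLarry backpack skeletal mesh, providing collision bodies and joints for its antenna, flaps and shoulder pieces.
# A central box body anchors everything, with box flaps, capsule shoulders and a three-segment capsule antenna chain.
# The antenna segments link in series with limited swing and twist.
# The flaps act as limited hinges with one swing axis and twist locked, while the shoulders rotate freely on all angular axes.
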PhysicsAsset /Figment_S10/Accessories/FORT_Backpacks/M_MED_BarbequeLarry_Backpack/Meshes/M_MED_BarbequeLarry_Backpack_Physics
bodies {
 antennae_01 [{shapeType=Capsule, length=1, radius=0.5, center=<0.1 0 0>, rotation=<rot -90 0 0>}]
 antennae_02 [{shapeType=Capsule, length=1, radius=0.5, center=<0.1 0 0>, rotation=<rot -90 0 0>}]
 antennae_03 [{shapeType=Capsule, length=1, radius=0.5, center=<0.1 0 0>, rotation=<rot -90 0 0>}]
 body [{shapeType=Box, extentX=28.9708, extentY=22.7712, extentZ=18.7469, center=<-1.5637 -1.7433 0>}]
 flap_l [{shapeType=Box, extentX=5.2827, extentY=7.801, extentZ=2.399, center=<3.539 0.1617 0.1238>}]
 flap_r [{shapeType=Box, extentX=5.6084, extentY=7.801, extentZ=2.399, center=<-3.6977 -0.1616 -0.1238>}]
 shoulder_l [{shapeType=Capsule, length=6.1657, radius=1.5822, center=<4.513 0 0>, rotation=<rot 90 0 0>}]
 shoulder_r [{shapeType=Capsule, length=6.1657, radius=1.5822, center=<-4.5087 0.0001 0>, rotation=<rot 90 0 0>}]
}
constraints [
 {bone1Name=antennae_01, bone2Name=body, swing1LimitDegrees=45, swing2LimitDegrees=45, twistLimitDegrees=15}
 {bone1Name=antennae_02, bone2Name=antennae_01, swing1LimitDegrees=45, swing2LimitDegrees=45, twistLimitDegrees=15}
 {bone1Name=antennae_03, bone2Name=antennae_02, swing1LimitDegrees=45, swing2LimitDegrees=45, twistLimitDegrees=15}
 {bone1Name=flap_r, bone2Name=body, swing1LimitDegrees=0, swing1Motion=Locked, swing2LimitDegrees=80, twistLimitDegrees=45, twistMotion=Locked}
 {bone1Name=flap_l, bone2Name=body, swing1LimitDegrees=0, swing1Motion=Locked, swing2LimitDegrees=80, twistLimitDegrees=45, twistMotion=Locked}
 {
  bone1Name=shoulder_l
  bone2Name=body
  swing1LimitDegrees=45
  swing1Motion=Free
  swing2LimitDegrees=45
  swing2Motion=Free
  twistLimitDegrees=45
  twistMotion=Free
 }
 {
  bone1Name=shoulder_r
  bone2Name=body
  swing1LimitDegrees=45
  swing1Motion=Free
  swing2LimitDegrees=45
  swing2Motion=Free
  twistLimitDegrees=45
  twistMotion=Free
 }
]
