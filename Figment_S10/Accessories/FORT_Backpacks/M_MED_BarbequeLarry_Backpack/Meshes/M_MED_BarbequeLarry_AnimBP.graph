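# Animation Blueprint for the BarbequeLarry backpack, adding secondary physics motion to the incoming pose.
# A body-bone AnimDynamics spring feeds a RigidBody node with an override physics asset and a skydiving force input.
# Spring alpha scales with the pawn's 2D speed and drops to zero while falling; opening the parachute resets dynamics.
# The flap_l and flap_r AnimDynamics chain is not connected to the output pose.
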AnimBlueprint /Figment_S10/Accessories/FORT_Backpacks/M_MED_BarbequeLarry_Backpack/Meshes/M_MED_BarbequeLarry_AnimBP
target_skeleton=/Game/Accessories/FORT_Backpacks/M_MED_BarbequeLarry_Backpack/Meshes/M_MED_BarbequeLarry_Backpack_Skeleton
graphs [
 AnimGraph {nodes [
   n1 = Anim Dynamics - Bone: flap_l(Alpha=1, ExternalForce="0,0,0") {
    bGravityOverrideInSimSpace=true
    bUseGravityOverride=true
    alphaBoolBlend {blendOption=Linear}
    boundBone {boneName=flap_l}
    componentLinearAccScale=<0.5 0.5 0.5>
    gravityOverride {z=-980}
    physicsBodyDefinitions [{
      sphereCollisionRadius=0
      boundBone {boneName=flap_l}
      boxExtents=<5 5 1>
      constraintSetup {angularLimitsMax=<0 10 0>, angularLimitsMin=<0 -50 0>, angularTarget=<0 0 0>}
      localJointOffset=<0.75 0 0>
     }]
   }
   Anim Dynamics - Bone: flap_r(ComponentPose=@n1, Alpha=1, ExternalForce="0,0,0") {
    bGravityOverrideInSimSpace=true
    bUseGravityOverride=true
    alphaBoolBlend {blendOption=Linear}
    boundBone {boneName=flap_r}
    componentLinearAccScale=<0.5 0.5 0.5>
    gravityOverride {z=-980}
    physicsBodyDefinitions [{
      sphereCollisionRadius=0
      boundBone {boneName=flap_r}
      boxExtents=<5 5 1>
      constraintSetup {angularLimitsMax=<0 10 0>, angularLimitsMin=<0 -50 0>, angularTarget=<0 0 0>}
      localJointOffset=<-0.75 0 0>
     }]
   }
   AnimGraphNode_SubInput_0 = Input Pose()
   n2 = Anim Dynamics - Bone: body(
    ComponentPose=Local To Component(LocalPose=@AnimGraphNode_SubInput_0)
    Alpha=Get bodyDynamics() -> (bodyDynamics:double)
    ExternalForce="0,0,0"
   ) {
    bGravityOverrideInSimSpace=true
    bLinearSpring=true
    bUseGravityOverride=true
    linearDampingOverride=0.99
    linearSpringConstant=100
    alphaBoolBlend {blendOption=Linear}
    alphaScaleBiasClamp {bInterpResult=true, bMapRange=true, inRange {max=500}}
    boundBone {boneName=body}
    componentAppliedLinearAccClamp=<5000 5000 5000>
    componentLinearAccScale=<0.5 0.5 0.5>
    componentLinearVelScale=<0.5 0.5 0.5>
    gravityOverride {z=-980}
    physicsBodyDefinitions [{
      sphereCollisionRadius=0
      boundBone {boneName=body}
      boxExtents=<15 10 15>
      constraintSetup {angularLimitsMax=<0 0 0>, angularLimitsMin=<0 0 0>, angularTarget=<0 0 0>, linearAxesMax=<0.5 0 0.5>, linearAxesMin=<-0.5 0 -0.5>}
      localJointOffset=<0.75 0 0>
     }]
   }
   n3 = RigidBody(
    ComponentPose=@n2
    ExternalForce=Break Fort Anim Input Skydiving External Force(FortAnimInput_SkydivingExternalForce=Make Fort Anim Input Skydiving External Force!() -> (FortAnimInput_SkydivingExternalForce:Fort Anim Input Skydiving External Force Structure) {is=title K2Node_MakeStruct}) -> (
     bUseSkydivingVectorForce=~
     SkydivingVectorForce=~
     HeadToPelvisDirection=~
     FloatingMultiplier=~
     FloatingAdditive=~
     DivingMultiplier=~
     DivingAdditive=~
     ParachutingMultiplier=~
     ParachutingAdditive=~
    ).SkydivingVectorForce
   ) {
    bOverrideWorldGravity=true
    overridePhysicsAsset=&/M_MED_BarbequeLarry_Backpack_Physics
    simulationSpace=BaseBoneSpace
    alphaBoolBlend {blendOption=Linear}
    baseBoneRef {boneName=root}
    componentAppliedLinearAccClamp=<3000 3000 3000>
    componentLinearAccScale=<0.4 0.4 0.4>
    overrideWorldGravity {z=-980}
   }
   Output Pose(Result=Component To Local(ComponentPose=@n3)) {layerGroup=None}
  ]}
 EventGraph {nodes [
   Event BlueprintUpdateAnimation()
   > n1 = Branch(Condition=Get bIsFalling())
   > Set bodyDynamics(bodyDynamics:double=0) -> (Output_Get:double)
   Set bodyDynamics(execute=@n1.else, bodyDynamics=Get PawnSpeed2D())
   Event OnParachuteOpened()
   > ResetDynamics(InTeleportType=TeleportPhysics)
   > ResetDynamics(InTeleportType=ResetPhysics)
   TryGetPawnOwner()
  ]}
]
variables [
 bodyDynamics {type=number, default=1}
 NewVar_0 {type=AnimInstance, default=null}
]
